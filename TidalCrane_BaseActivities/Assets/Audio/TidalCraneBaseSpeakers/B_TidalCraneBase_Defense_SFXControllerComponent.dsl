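; Cosmetic audio controller component for TidalCrane base defense and reclamation events, driving the base's alarms, PA voice lines and speaker bus.
; It listens for both events through the GameplayEventRouter, filters by base name, updates BaseState and stops alarm and outside VO via EventCompleted.
; Music encounter volume enter/exit manages alarms, a mutator and outside VO.
; Audio gameplay volume toggles switch the bus, diegesis mix and listener flag gating progress-based defense VO.
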(event EventBeginPlay
  (bind _returnvalue (Components|GetOwner))
  (|CallParentFunction)
  (Networking|CanExecuteCosmeticEvents
    (:True
      (Utilities|IsValid _returnvalue
        (:Is%20Valid
          (bind _output_get (Variables|Default|SetOwningActor _returnvalue))
          (bind _returnvalue_2 (GameplayEventRouter|GetEventRouter "/Script/FortniteGame.FortPlayspace" _output_get))
          (Utilities|FlowControl|Sequence
            (:then_0
              (CallFunction|GetBaseDetails))
            (:then_1
              ;[0 0.7278 0.8854 0.5] Defense
              (bind (_asynctaskproxy _payload) (GameplayEventRouter|StartListeningToEvent _returnvalue_2 "/TidalCrane_BaseActivities/Gameplay/TidalCrane_Base_DefenseData.TidalCrane_Base_DefenseData")
                (:OnEventReceived
                  (if (Utilities|String|EqualExactly(String) (Variables|Default|GetBaseName) (Variables|/TidalCrane_BaseActivities/Gameplay/B_TidalCrane_BaseLocator.B_TidalCrane_BaseLocator_C|GetBaseName (Variables|/TidalCrane_BaseActivities/Gameplay/B_TidalCrane_MilitaryActivity_SpawnManager.B_TidalCrane_MilitaryActivity_SpawnManager_C|GetBaseLocator (Utilities|Struct|BreakTidalCraneBaseDefenseData _payload))))
                    (Variables|Default|SetBaseState (.EventState_10_5AACBE6344F0C1F9283B6F996EE80287 (Utilities|Struct|BreakTidalCraneBaseDefenseData _payload)))
                    (switch Utilities|FlowControl|Switch|SwitchonTidalCrane_Bases_Enum (.EventState_10_5AACBE6344F0C1F9283B6F996EE80287 (Utilities|Struct|BreakTidalCraneBaseDefenseData _payload))
                      (:NewEnumerator2
                        (CallFunction|DefensePlayVO (.BugsKilled_6_97A8A9E7411F76A1DADB4DAA60527F22 (Utilities|Struct|BreakTidalCraneBaseDefenseData _payload)) (.MaxBugs_27_E3A9792F47777D86969E85A4A891A09F (Utilities|Struct|BreakTidalCraneBaseDefenseData _payload))))
                      (:NewEnumerator5
                        _)
                      (:NewEnumerator6
                        _))
                    (CallFunction|EventCompleted)
                    (CallFunction|DefensePlayVO (.BugsKilled_6_97A8A9E7411F76A1DADB4DAA60527F22 (Utilities|Struct|BreakTidalCraneBaseDefenseData _payload)) (.MaxBugs_27_E3A9792F47777D86969E85A4A891A09F (Utilities|Struct|BreakTidalCraneBaseDefenseData _payload)))))))
            (:then_2
              ;[0 0.7278 0.8854 0.5] Reclamation
              (bind (_asynctaskproxy_2 _payload_2) (GameplayEventRouter|StartListeningToEvent _returnvalue_2 "/TidalCrane_BaseActivities/Gameplay/TidalCrane_Base_Reclamation_Data.TidalCrane_Base_Reclamation_Data")
                (:OnEventReceived
                  (if (Utilities|String|EqualExactly(String) (Variables|Default|GetBaseName) (Variables|/TidalCrane_BaseActivities/Gameplay/B_TidalCrane_BaseLocator.B_TidalCrane_BaseLocator_C|GetBaseName (Variables|/TidalCrane_BaseActivities/Gameplay/B_TidalCrane_MilitaryActivity_SpawnManager.B_TidalCrane_MilitaryActivity_SpawnManager_C|GetBaseLocator (Utilities|Struct|BreakTidalCraneBaseReclamationData _payload_2))))
                    (bind _output_get_2 (Variables|Default|SetBaseState (.EventState_5_781D3B7C4B9A8EFDCFC0F5832D1CE7C1 (Utilities|Struct|BreakTidalCraneBaseReclamationData _payload_2))))
                    (switch Utilities|FlowControl|Switch|SwitchonTidalCrane_Bases_Enum _output_get_2
                      (:NewEnumerator5
                        _)
                      (:NewEnumerator6
                        _))
                    (CallFunction|EventCompleted)))))
            (:then_3
              (CallFunction|ActivateMusicEncounterVolumeBindings))
            (:then_4
              (Audio|Bus|StartAudioBus "/TidalCrane_BaseActivities/Assets/Audio/TidalCraneBaseSpeakers/AB_TidalCraneBase_SpeakerSystem.AB_TidalCraneBase_SpeakerSystem"))))))))

(event Custom|ActivateMusicEncounterVolumeBindings
  ;[0.5751 0.1986 0.8854 0.5] This spawns and stops the audio based on the player in the encounter or not
  (bind _volume_me_threshold (Variables|Default|GetVolume_ME_Threshold)
    (:then
      (Events|AssignOnListenerEnterEvent _volume_me_threshold (delegate Custom|OnListenerEnterEvent_Event_ME))
      (Events|AssignOnListenerExitEvent _volume_me_threshold (delegate Custom|OnListenerExitEvent_Event_ME)))
    (:else)))

(event Custom|OnListenerExitEvent_Event_ME
  ;[0.5751 0.1986 0.8854 0.5] This spawns and stops the audio based on the player in the encounter or not
  (Utilities|FlowControl|Sequence
    (:then_0
      (bind _basealarmsystem (Variables|Default|GetBaseAlarmSystem)
        (:then
          (Audio|Components|Audio|FadeOut _basealarmsystem 4.0 0.0 "Linear" :!class /Script/Engine.AudioComponent))
        (:else)))
    (:then_1
      (Audio|Mutator|RemoveMutator (AudioEngineSubsystems|GetAudioMutatorsSubsystem) (Variables|Default|GetEventMutator))
      (bind _outsidebasevo (Variables|Default|GetOutsideBaseVO)
        (:then
          (Audio|Components|Audio|Stop _outsidebasevo))
        (:else)))))

(event Custom|OnListenerEnterEvent_Event_ME
  ;[0.5751 0.1986 0.8854 0.5] This spawns and stops the audio based on the player in the encounter or not
  (Utilities|FlowControl|Sequence
    (:then_0
      (CallFunction|SpawnAlertAudio))
    (:then_1
      (Audio|Mutator|AddMutator (AudioEngineSubsystems|GetAudioMutatorsSubsystem) (Variables|Default|GetEventMutator)))
    (:then_2
      (switch Utilities|FlowControl|Switch|SwitchonTidalCrane_Bases_Enum (Variables|Default|GetBaseState)
        (:NewEnumerator2
          (Variables|Default|GetOutsideBaseVO
            (:then)
            (:else
              (bind _returnvalue (Audio|SpawnSoundatLocation "/TidalCrane_BaseActivities/Assets/Audio/TidalCraneBaseSpeakers/MSS_TidalCraneBase_PASystem_VO_OutsideBase.MSS_TidalCraneBase_PASystem_VO_OutsideBase" (Transformation|GetActorLocation (Variables|Default|GetBaseLocator))))
              (Variables|Default|SetOutsideBaseVO _returnvalue))))))))

(event Custom|SpawnAlertAudio
  ;[0.5751 0.1986 0.8854 0.5] Creation of the alarm if the event is base and not completed
  (CallFunction|GetBaseDetails)
  (if (and (Utilities|Enum|Equal(Enum) (Variables|Default|GetBaseState) "NewEnumerator2") (Utilities|Enum|Equal(Enum) (Variables|Default|GetBase%20Type) "NewEnumerator1"))
    (bind _owningactor (Variables|Default|GetOwningActor)
      (:then
        (Variables|Default|GetBaseAlarmSystem
          (:then)
          (:else
            (bind _returnvalue (Audio|SpawnSoundAttached "/TidalCrane_BaseActivities/Assets/Audio/TidalCraneBaseSpeakers/MSS_TidalCraneBase_PASystem_Alarms.MSS_TidalCraneBase_PASystem_Alarms" (Variables|/Script/Engine.Actor|GetRootComponent _owningactor)))
            (bind _output_get (Variables|Default|SetBaseAlarmSystem _returnvalue))
            (Audio|Parameter|SetBooleanParameter _output_get "bIsReclamationEvent" (Utilities|Enum|Equal(Enum) (Variables|Default|GetBase%20Type) "NewEnumerator3") :!class /Script/AudioExtensions.AudioParameterControllerInterface))))
      (:else))
    (else
      (bind _basealarmsystem (Variables|Default|GetBaseAlarmSystem)
        (:then
          (Audio|Components|Audio|FadeOut _basealarmsystem 0.5 0.0 "Linear" :!class /Script/Engine.AudioComponent))
        (:else)))))

(event AudioGameplayVolume|EventOnToggledOn
  ;[0 0.7278 0.8854 0.5] Listener Enter/Exit
  (CallFunction|GetBaseDetails)
  (if (Utilities|Enum|Equal(Enum) (Variables|Default|GetBaseState) "NewEnumerator2")
    (Utilities|FlowControl|Sequence
      (:then_0
        (Variables|Default|SetbListenerInVolume true))
      (:then_1
        (Audio|Modulation|ActivateControlBusMix "/TidalCrane_BaseActivities/Assets/Audio/TidalCraneBaseSpeakers/CBM_TidalCraneBase_Defense_AlarmDiegesisControl.CBM_TidalCraneBase_Defense_AlarmDiegesisControl")
        (Audio|Bus|StartAudioBus "/TidalCrane_BaseActivities/Assets/Audio/TidalCraneBaseSpeakers/AB_TidalCraneBase_SpeakerSystem.AB_TidalCraneBase_SpeakerSystem"))
      (:then_2
        _)
      (:then_3
        (Utilities|FlowControl|RetriggerableDelay 5.0))
      (:then_4
        (bind _outsidebasevo (Variables|Default|GetOutsideBaseVO)
          (:then
            (Audio|Components|Audio|FadeOut _outsidebasevo 0.5 0.0 "Linear" :!class /Script/Engine.AudioComponent))
          (:else))))
    (Utilities|FlowControl|DoOnce false
      (:Completed
        (CallFunction|DefensePlayVO 0 20 true)))))

(event AudioGameplayVolume|EventOnToggledOff
  ;[0 0.7278 0.8854 0.5] Listener Enter/Exit
  (Audio|Modulation|DeactivateControlBusMix "/TidalCrane_BaseActivities/Assets/Audio/TidalCraneBaseSpeakers/CBM_TidalCraneBase_Defense_AlarmDiegesisControl.CBM_TidalCraneBase_Defense_AlarmDiegesisControl")
  (Audio|Bus|StopAudioBus "/TidalCrane_BaseActivities/Assets/Audio/TidalCraneBaseSpeakers/AB_TidalCraneBase_SpeakerSystem.AB_TidalCraneBase_SpeakerSystem")
  (Variables|Default|SetbListenerInVolume false)
  (switch Utilities|FlowControl|Switch|SwitchonTidalCrane_Bases_Enum (Variables|Default|GetBaseState)
    (:NewEnumerator2
      (Variables|Default|GetOutsideBaseVO
        (:then)
        (:else
          (bind _returnvalue (Audio|SpawnSoundatLocation "/TidalCrane_BaseActivities/Assets/Audio/TidalCraneBaseSpeakers/MSS_TidalCraneBase_PASystem_VO_OutsideBase.MSS_TidalCraneBase_PASystem_VO_OutsideBase" (Transformation|GetActorLocation (Variables|Default|GetBaseLocator))))
          (Variables|Default|SetOutsideBaseVO _returnvalue))))))

(event Custom|EventCompleted
  (Utilities|FlowControl|Sequence
    (:then_0
      (bind _basealarmsystem (Variables|Default|GetBaseAlarmSystem)
        (:then
          (Audio|Components|Audio|Stop _basealarmsystem))
        (:else)))
    (:then_1
      (bind _outsidebasevo (Variables|Default|GetOutsideBaseVO)
        (:then
          (Audio|Components|Audio|Stop _outsidebasevo))
        (:else)))))

(fn ChooseBaseName
  (return (Utilities|Array|MakeArray "Alpha" "Beta" "Charlie")))

(fn DefensePlayVO
  (params (KilledBugs :type Integer) (MaxBugs :type Integer) (bIsStart :type Boolean))
  (Variables|Default|GetbListenerInVolume
    (:then
      (LocalVariables|SetPercentComplete (/ (Math|Conversions|ToFloat(Integer) KilledBugs) (Math|Conversions|ToFloat(Integer) MaxBugs)))
      (LocalVariables|SetLocation (Transformation|GetActorLocation (Variables|Default|GetOwningActor)))
      (if (Math|Integer|InRange(Integer) (Variables|GetKilledBugs) 0 1)
        (Audio|PlaySoundatLocation "/TidalCrane_BaseActivities/Assets/Audio/TidalCraneBaseSpeakers/MSS_TidalCraneBase_PASystem_VO_StartEvent.MSS_TidalCraneBase_PASystem_VO_StartEvent" (LocalVariables|GetLocation))
        (elif (Math|Integer|InRange(Integer) (Variables|GetKilledBugs) (- (/ (Variables|GetMaxBugs) 2) 1) (+ (/ (Variables|GetMaxBugs) 2) 1))
          (Audio|PlaySoundatLocation "/TidalCrane_BaseActivities/Assets/Audio/TidalCraneBaseSpeakers/MSS_TidalCraneBase_PASystem_VO_MidEvent.MSS_TidalCraneBase_PASystem_VO_MidEvent" (LocalVariables|GetLocation))
          (elif (>= (LocalVariables|GetPercentComplete) 1.0)
            (Audio|PlaySoundatLocation "/TidalCrane_BaseActivities/Assets/Audio/TidalCraneBaseSpeakers/MSS_TidalCraneBase_PASystem_VO_EndEvent.MSS_TidalCraneBase_PASystem_VO_EndEvent" (LocalVariables|GetLocation))))))
    (:else)))

(fn GetBaseDetails
  (bind _baselocator (Variables|Default|GetBaseLocator))
  (bind _baselocator_2 (Variables|Default|GetBaseLocator)
    (:then
      (if (Utilities|String|EqualExactly(String) (Variables|Default|GetBaseName) (Variables|/TidalCrane_BaseActivities/Gameplay/B_TidalCrane_BaseLocator.B_TidalCrane_BaseLocator_C|GetBaseName _baselocator_2))
        (Variables|Default|SetBase%20Type (Variables|/TidalCrane_BaseActivities/Gameplay/B_TidalCrane_BaseLocator.B_TidalCrane_BaseLocator_C|GetBaseType _baselocator))
        (Variables|Default|SetBaseState (Variables|/TidalCrane_BaseActivities/Gameplay/B_TidalCrane_BaseLocator.B_TidalCrane_BaseLocator_C|GetBaseState _baselocator))))
    (:else)))
