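; Blueprint B_TidalCraneBase_Speaker, a speaker actor among the audio assets for the Tidal Crane base.
; When its linked actor is destroyed, the Fort link-to-actor event destroys this speaker actor, ignoring the damage causer.
; The construction script is empty, and no sound playback logic appears in the graphs.
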(event OnLinkedActorDestroyed(FortLinkToActor)
  (params (DamageCauser :type Actor%20Object%20Reference))
  (Actor|DestroyActor))

(fn ConstructionScript)
