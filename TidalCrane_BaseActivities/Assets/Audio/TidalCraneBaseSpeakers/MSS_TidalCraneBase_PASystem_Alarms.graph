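# Looping mono MetaSound source for the TidalCrane base PA system alarms, played with outside-speaker attenuation.
# Two looping wave players, distant and close alarm, feed a crossfade node whose value comes from attenuation distance, range-mapped and shaped by a defense wavetable bank curve.
# The AlarmDiegesisControl modulator, smoothed by InterpTo, splits the result between the source output and an audio bus writer, which receives the inverse gain.
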MetaSoundSource /TidalCrane_BaseActivities/Assets/Audio/TidalCraneBaseSpeakers/MSS_TidalCraneBase_PASystem_Alarms
inputs [
 UE.Attenuation.Distance {type=Float}
 UE.Source.OnPlay {type=Trigger}
]
interfaces [UE.OutputFormat.Mono:1.0, UE.Source:1.1]
outputs ["UE.OutputFormat.Mono.Audio:0" {type=Audio}]
pages [Default {
  input_defaults {UE.Attenuation.Distance=17000, UE.Source.OnPlay=null}
  nodes [
   WavePlayer_0 = UE.Wave Player:Mono(Play=^UE.Source.OnPlay) {also {
     defaults {
      Loop="<MetaSoundBuilderResult.SUCCEEDED: 0>"
      Wave Asset=&/TidalCraneBase_Alarm_Distant
     }
     version [1,1]
    }}
   WavePlayer_1 = UE.Wave Player:Mono(Play=^UE.Source.OnPlay) {also {
     defaults {
      Loop="<MetaSoundBuilderResult.SUCCEEDED: 0>"
      Wave Asset=&/TidalCraneBase_Alarm_Close
     }
     version [1,1]
    }}
   CrossfadeTriggerRouteAudio2_0 = "Crossfade.Trigger Route (Audio, 2):Audio"(
    In 0=@WavePlayer_0.Out Mono
    In 1=@WavePlayer_1.Out Mono
    Crossfade Value=UE.WaveTableBankEvaluate(Input=MapRange.MapRange:Float(In=^UE.Attenuation.Distance) {also {defaults {Out Range A=4000, Out Range B=16000}}}.Out Value) {also {defaults {Index=2, WaveTableBank=/TidalCrane_BaseActivities/Assets/Audio/TidalCraneMixing/WTB_TidalCraneBase_Defense}}}.Output
   )
   n1 = UE.InterpTo:Audio(Target=Modulation.GetModulatorValue() {also {defaults {Modulator=&/CB_TidalCraneBase_Defense_AlarmDiegesisControl}}}.Out) {also {defaults {Interp Time=1}}}
   AudioBusWriter1_0 = "UE.Audio Bus Writer (1)"(In 0=UE.Multiply:Audio by Float(
     PrimaryOperand=@CrossfadeTriggerRouteAudio2_0.Out
     AdditionalOperands=UE.Subtract:Float(AdditionalOperands=@n1.Value) {also {defaults {PrimaryOperand=1}}}.Out
    ).Out)
   Multiply_0 = UE.Multiply:Audio by Float(PrimaryOperand=@CrossfadeTriggerRouteAudio2_0.Out, AdditionalOperands=@n1.Value)
  ]
  links [{dst="@UE.OutputFormat.Mono.Audio:0", src=Multiply_0.Out}]
 }]
properties {
 assetImportData=&self:AssetImportData
 attenuationSettings=&/ATT_TidalCraneBase_OutsideSpeaker
 bLooping=true
 duration=10000
 maxDistance=18000
 sampleRate=48000
 soundClassObject=&Fort_SFX_InGame
 volume=2
}
