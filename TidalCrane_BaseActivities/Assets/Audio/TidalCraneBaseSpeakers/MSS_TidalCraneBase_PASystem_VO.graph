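# MetaSound source for TidalCrane base PA system voice-over, played as a mono one-shot.
# On play, a random clip is picked from the In Array input and fed to a mono wave player, whose completion ends the sound.
# Output is also sent at full level to the TidalCrane base speaker-system audio bus.
# It uses the in-game SFX sound class and keeps playing when silent.
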MetaSoundSource /TidalCrane_BaseActivities/Assets/Audio/TidalCraneBaseSpeakers/MSS_TidalCraneBase_PASystem_VO
inputs [
 In Array {type=WaveAsset:Array}
 UE.Source.OnPlay {type=Trigger}
]
interfaces [UE.OutputFormat.Mono:1.0, UE.Source.OneShot:1.0, UE.Source:1.1]
outputs [
 "UE.OutputFormat.Mono.Audio:0" {type=Audio}
 UE.Source.OneShot.OnFinished {type=Trigger}
]
pages [Default {
  input_defaults {In Array=null, UE.Source.OnPlay=null}
  nodes [
   ArrayRandomGet_0 = Array.Random Get:WaveAsset:Array(In Array=^In Array, Next=^UE.Source.OnPlay)
   WavePlayer_0 = UE.Wave Player:Mono(Play=@ArrayRandomGet_0.On Next, Wave Asset=@ArrayRandomGet_0.Value) {also {version [1,1]}}
  ]
  links [{dst="@UE.Source.OneShot.OnFinished", src=WavePlayer_0.On Finished}, {dst="@UE.OutputFormat.Mono.Audio:0", src=WavePlayer_0.Out Mono}]
 }]
properties {
 assetImportData=&self:AssetImportData
 sampleRate=48000
 soundClassObject=&Fort_SFX_InGame
 virtualizationMode=PlayWhenSilent
 busSends [{audioBus=&/AB_TidalCraneBase_SpeakerSystem, sendLevel=1}]
}
