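# MetaSound Source preset that plays mid-event base defense Fortlish VO lines for the Tidal Crane base PA system.
# It inherits its graph from MSS_TidalCraneBase_PASystem_VO and overrides the In Array input with twelve mid-event wave variants.
# It is a mono one-shot source with base submix and submix sends disabled, feeding the base speaker system audio bus instead.
# It keeps playing when silent through its virtualization mode.
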MetaSoundSource /TidalCrane_BaseActivities/Assets/Audio/TidalCraneBaseSpeakers/MSS_TidalCraneBase_PASystem_VO_MidEvent
preset_parent=&/MSS_TidalCraneBase_PASystem_VO
inputs [
 In Array {type=WaveAsset:Array}
 UE.Source.OnPlay {type=Trigger}
]
interfaces [UE.OutputFormat.Mono:1.0, UE.Source.OneShot:1.0, UE.Source:1.1]
outputs [
 "UE.OutputFormat.Mono.Audio:0" {type=Audio}
 UE.Source.OneShot.OnFinished {type=Trigger}
]
pages [Default {input_defaults {
   UE.Source.OnPlay=null
   In Array [
    /TidalCrane_BaseActivities/Assets/VO/BaseDefenseFortlish/BaseDefense_Event_Mid_FortlishVO_12
    /TidalCrane_BaseActivities/Assets/VO/BaseDefenseFortlish/BaseDefense_Event_Mid_FortlishVO_11
    /TidalCrane_BaseActivities/Assets/VO/BaseDefenseFortlish/BaseDefense_Event_Mid_FortlishVO_10
    /TidalCrane_BaseActivities/Assets/VO/BaseDefenseFortlish/BaseDefense_Event_Mid_FortlishVO_09
    /TidalCrane_BaseActivities/Assets/VO/BaseDefenseFortlish/BaseDefense_Event_Mid_FortlishVO_08
    /TidalCrane_BaseActivities/Assets/VO/BaseDefenseFortlish/BaseDefense_Event_Mid_FortlishVO_07
    /TidalCrane_BaseActivities/Assets/VO/BaseDefenseFortlish/BaseDefense_Event_Mid_FortlishVO_06
    /TidalCrane_BaseActivities/Assets/VO/BaseDefenseFortlish/BaseDefense_Event_Mid_FortlishVO_05
    /TidalCrane_BaseActivities/Assets/VO/BaseDefenseFortlish/BaseDefense_Event_Mid_FortlishVO_04
    /TidalCrane_BaseActivities/Assets/VO/BaseDefenseFortlish/BaseDefense_Event_Mid_FortlishVO_03
    /TidalCrane_BaseActivities/Assets/VO/BaseDefenseFortlish/BaseDefense_Event_Mid_FortlishVO_02
    /TidalCrane_BaseActivities/Assets/VO/BaseDefenseFortlish/BaseDefense_Event_Mid_FortlishVO_01
   ]
  }}]
properties {
 assetImportData=&self:AssetImportData
 bEnableBaseSubmix=false
 bEnableSubmixSends=false
 sampleRate=48000
 virtualizationMode=PlayWhenSilent
 busSends [{audioBus=&/AB_TidalCraneBase_SpeakerSystem, sendLevel=1}]
}
