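# Looping mono MetaSound source for TidalCrane base outdoor PA speakers, repeating Fortlish base defense start announcements.
# Passes eight VO waves into the nested PASystem_VO MetaSound, first triggered on play.
# Each finished line starts a random delay, fed back through a trigger variable to replay.
# Uses outside-speaker attenuation, NPC VO sound class, play-when-silent virtualization, and volume modulation from the defense alarm diegesis control bus.
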MetaSoundSource /TidalCrane_BaseActivities/Assets/Audio/TidalCraneBaseSpeakers/MSS_TidalCraneBase_PASystem_VO_OutsideBase
inputs [
 In Array {type=WaveAsset:Array}
 UE.Source.OnPlay {type=Trigger}
]
interfaces [UE.OutputFormat.Mono:1.0, UE.Source:1.1]
outputs ["UE.OutputFormat.Mono.Audio:0" {type=Audio}]
pages [Default {
  input_defaults {
   UE.Source.OnPlay=null
   In Array [
    /TidalCrane_BaseActivities/Assets/VO/BaseDefenseFortlish/BaseDefense_Event_Start_FortlishVO_01
    /TidalCrane_BaseActivities/Assets/VO/BaseDefenseFortlish/BaseDefense_Event_Start_FortlishVO_02
    /TidalCrane_BaseActivities/Assets/VO/BaseDefenseFortlish/BaseDefense_Event_Start_FortlishVO_03
    /TidalCrane_BaseActivities/Assets/VO/BaseDefenseFortlish/BaseDefense_Event_Start_FortlishVO_04
    /TidalCrane_BaseActivities/Assets/VO/BaseDefenseFortlish/BaseDefense_Event_Start_FortlishVO_05
    /TidalCrane_BaseActivities/Assets/VO/BaseDefenseFortlish/BaseDefense_Event_Start_FortlishVO_06
    /TidalCrane_BaseActivities/Assets/VO/BaseDefenseFortlish/BaseDefense_Event_Start_FortlishVO_07
    /TidalCrane_BaseActivities/Assets/VO/BaseDefenseFortlish/BaseDefense_Event_Start_FortlishVO_08
   ]
  }
  nodes [
   n1 = VariableDeferredAccessor.Trigger(Variable=InitVariable.Trigger().Variable)
   TriggerAnyTriggerAny2_0 = "TriggerAny.Trigger Any (2)"(In 0=^UE.Source.OnPlay, In 1=@n1.Value)
   MSS_TidalCraneBase_PASystem_VO_0 = &/MSS_TidalCraneBase_PASystem_VO(In Array=^In Array, UE.Source.OnPlay=@TriggerAnyTriggerAny2_0.Out)
   n2 = UE.RandomTime(Next=@MSS_TidalCraneBase_PASystem_VO_0.UE.Source.OneShot.OnFinished) {also {defaults {Max=20, Min=10}}}
   TriggerDelay_0 = UE.Trigger Delay(In=@n2.On Next, Delay Time=@n2.Value)
   VariableMutator.Trigger(Value=@TriggerDelay_0.Out, Variable=@n1.Variable)
  ]
  links [{dst="@UE.OutputFormat.Mono.Audio:0", src=MSS_TidalCraneBase_PASystem_VO_0.UE.OutputFormat.Mono.Audio:0}]
 }]
properties {
 assetImportData=&self:AssetImportData
 attenuationSettings=&/ATT_TidalCraneBase_OutsideSpeaker
 bLooping=true
 duration=10000
 maxDistance=18000
 sampleRate=48000
 soundClassObject=&Fort_VO_NPC
 virtualizationMode=PlayWhenSilent
 volume=1.5
 modulationSettings {
  volumeRouting=Union
  volumeModulationDestination {
   bEnableModulation=true
   modulators [&/CB_TidalCraneBase_Defense_AlarmDiegesisControl]
  }
 }
}
